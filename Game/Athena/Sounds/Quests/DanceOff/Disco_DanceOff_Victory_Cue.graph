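# Sound cue for the victory music of the Disco DanceOff quest in the Athena game mode.
# A single wave player plays the Disco_DanceOff_Victory wave with no mixing, looping or variation.
# It is slightly boosted, routed to the quest music sound class, and spatialized through a single-area DanceOff attenuation asset with a limited audible range.
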SoundCue /Game/Athena/Sounds/Quests/DanceOff/Disco_DanceOff_Victory_Cue
root=WavePlayer_0
nodes [{id=WavePlayer_0, properties {soundWaveAssetPtr=&/Disco_DanceOff_Victory}}]
properties {
 attenuationSettings=&/DanceOff_SingleArea_Att
 maxDistance=1100
 soundClassObject=/Game/Sounds/FortSoundClassesAndModes/Fort_Music_Quest_PSM
 volumeMultiplier=1.2
}
